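# Editor utility that batch-applies LOD, collision and Nanite settings to selected static meshes, then checks out and saves them.
# AssignLODsToStaticMesh builds Nanite settings once before the loop; PreserveTangentSettings uses stored settings instead of the other inputs, reading tangents before any mesh is assigned.
# Each mesh gets a LOD group if it has two or fewer LODs, quality-level minimum LODs, collision LOD and screen sizes chosen by folder-path keyword.
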EditorUtilityBlueprint /Game/ArtTools/Environment/BlueprintSetup/AAU_ApplyLODSettings
graphs [
 AssignLODsToStaticMesh {nodes [
   n1 = AssignLODsToStaticMesh!() -> (
    then
    Enable Nanite:bool=true
    PreserveTangentSettings:bool=true
    Explicit Tangents:bool=~
    Fallback Target:ENaniteFallbackTarget Enum=RelativeError
    Fallback Relative Error:float=~
   ) {is=title K2Node_FunctionEntry}
   > n2 = Branch(Condition=@n1.PreserveTangentSettings)
   n3 = GetSelectedAssets(execute=[@n10.then, @n9.then])
   > n4 = Filter by Class!(
    execute
    TargetArray:Array of Object References=@n3
    ObjectClass:Object Class Reference=&StaticMesh
    FilterType:EEditorScriptingFilterType Enum
   ) -> (then, ReturnValue:Array of Static Mesh Object References)
   > n5 = For Each Loop(Array:Array of Static Mesh Object References=@n4) -> (Array Element:Static Mesh Object Reference)
   > n6 = Set L_MeshToModify(L_MeshToModify:Static Mesh Object Reference=@n5) -> (Output_Get:Static Mesh Object Reference)
   n7 = GetNaniteSettings!(self:Static Mesh Editor Subsystem Object Reference=Get StaticMeshEditorSubsystem(), StaticMesh:Static Mesh Object Reference=@n6) -> (ReturnValue:Mesh Nanite Settings Structure)
   n8 = Break Mesh Nanite Settings!(MeshNaniteSettings:"Mesh Nanite Settings Structure (by ref)"=Get STR_EnvNaniteSettings() -> (STR_EnvNaniteSettings:Mesh Nanite Settings Structure)) -> (
    bEnabled:bool=~
    bExplicitTangents:bool=~
    bLerpUVs:bool=~
    bSeparable:bool=~
    bVoxelNDF:bool=~
    bVoxelOpacity:bool=~
    ShapePreservation:ENaniteShapePreservation Enum
    PositionPrecision:int
    NormalPrecision:int
    TangentPrecision:int
    BoneWeightPrecision:int
    KeepPercentTriangles:float=~
    TrimRelativeError:float=~
    GenerateFallback:ENaniteGenerateFallback Enum
    FallbackTarget:ENaniteFallbackTarget Enum
    FallbackPercentTriangles:float=~
    FallbackRelativeError:float=~
    MaxEdgeLengthFactor:float=~
    NumRays:int
    VoxelLevel:int
    RayBackUp:float=~
    DisplacementUVChannel:int
   ) {is=title K2Node_BreakStruct}
   n9 = Set STR_EnvNaniteSettings(
    execute=@n2.then
    STR_EnvNaniteSettings:Mesh Nanite Settings Structure=Make Mesh Nanite Settings!(
     bEnabled:bool=@n8
     bExplicitTangents:bool=Break Mesh Nanite Settings(MeshNaniteSettings=@n7) -> (
      bEnabled=~
      bExplicitTangents=~
      bLerpUVs=~
      bSeparable=~
      bVoxelNDF=~
      bVoxelOpacity=~
      KeepPercentTriangles=~
      TrimRelativeError=~
      FallbackPercentTriangles=~
      FallbackRelativeError=~
      MaxEdgeLengthFactor=~
      RayBackUp=~
     ).bExplicitTangents
     bLerpUVs:bool=@n8.bLerpUVs
     bSeparable:bool=@n8.bSeparable
     bVoxelNDF:bool=@n8.bVoxelNDF
     bVoxelOpacity:bool=@n8.bVoxelOpacity
     ShapePreservation:ENaniteShapePreservation Enum=@n8.ShapePreservation
     PositionPrecision:int=@n8.PositionPrecision
     NormalPrecision:int=@n8.NormalPrecision
     TangentPrecision:int=@n8.TangentPrecision
     BoneWeightPrecision:int=@n8.BoneWeightPrecision
     KeepPercentTriangles:float=@n8.KeepPercentTriangles
     TrimRelativeError:float=@n8.TrimRelativeError
     GenerateFallback:ENaniteGenerateFallback Enum=@n8.GenerateFallback
     FallbackTarget:ENaniteFallbackTarget Enum=@n8.FallbackTarget
     FallbackPercentTriangles:float=@n8.FallbackPercentTriangles
     FallbackRelativeError:float=@n8.FallbackRelativeError
     MaxEdgeLengthFactor:float=@n8.MaxEdgeLengthFactor
     NumRays:int=@n8.NumRays
     VoxelLevel:int=@n8.VoxelLevel
     RayBackUp:float=@n8.RayBackUp
     DisplacementUVChannel:int=@n8.DisplacementUVChannel
    ) -> (MeshNaniteSettings:Mesh Nanite Settings Structure) {is=title K2Node_MakeStruct}
   ) -> (Output_Get:Mesh Nanite Settings Structure)
   n10 = Set STR_EnvNaniteSettings(
    execute=@n2.else
    STR_EnvNaniteSettings=Make Mesh Nanite Settings(
     bEnabled=@n1
     bExplicitTangents=@n1.Explicit Tangents
     bLerpUVs=True
     bSeparable=False
     bVoxelNDF=True
     bVoxelOpacity=False
     ShapePreservation=None
     PositionPrecision=-2147483648
     NormalPrecision=-1
     TangentPrecision=-1
     BoneWeightPrecision=-1
     KeepPercentTriangles=1
     TrimRelativeError=0
     GenerateFallback=PlatformDefault
     FallbackTarget=@n1.Fallback Target
     FallbackPercentTriangles=1
     FallbackRelativeError=@n1.Fallback Relative Error
     MaxEdgeLengthFactor=0
     NumRays=64
     VoxelLevel=0
     RayBackUp=0
     DisplacementUVChannel=0
    )
   )
   Break Source Control State!(SourceControlState:"Source Control State Structure (by ref)"=QueryFileState!(execute, InFile:String, bSilent:bool=false, bUseSourceControlStateCache:bool=false) -> (then, ReturnValue:Source Control State Structure)) -> (
    Filename:String
    bIsValid:bool=~
    bIsUnknown:bool=~
    bCanCheckIn:bool=~
    bCanCheckOut:bool=~
    bIsCheckedOut:bool=~
    bIsCurrent:bool=~
    bIsSourceControlled:bool=~
    bIsAdded:bool=~
    bIsDeleted:bool=~
    bIsIgnored:bool=~
    bCanEdit:bool=~
    bCanDelete:bool=~
    bIsModified:bool=~
    bCanAdd:bool=~
    bIsConflicted:bool=~
    bCanRevert:bool=~
    bIsCheckedOutOther:bool=~
    CheckedOutOther:String
    bIsCheckedOutInOtherBranch:bool=~
    bIsModifiedInOtherBranch:bool=~
    PreviousUser:String
   ) {is=title K2Node_BreakStruct}
   n11 = Get STR_EnvNaniteSettings()
   SetNaniteSettings!(
    execute=@n25.then_3
    self:Static Mesh Editor Subsystem Object Reference=Get StaticMeshEditorSubsystem()
    StaticMesh:Static Mesh Object Reference=Get L_MeshToModify() -> (L_MeshToModify:Static Mesh Object Reference)
    NaniteSettings:Mesh Nanite Settings Structure=@n11
    bApplyChanges:bool=true
   ) -> (then)
   PrintString(execute=@n23.else, InString="To String (Text)"(InText=LastErrorMsg!() -> (ReturnValue:Text)))
   Show Message Dialog(MessageType=Ok, DefaultValue=No, MessageCategory=Warning)
   n12 = Get StaticMeshEditorSubsystem()
   n13 = Get L_MeshToModify()
   SetLodScreenSizes!(
    execute=@n28.then
    self:Static Mesh Editor Subsystem Object Reference=@n12
    StaticMesh:Static Mesh Object Reference=@n13
    ScreenSizes:"Array of Float (single-precision)s"=Get L_Prop_ScreenSizes() -> (L_Prop_ScreenSizes:"Array of Float (single-precision)s")
   ) -> (then, ReturnValue:bool)
   SetMinLODForQualityLevels!(
    execute=[@n24.then, @n27.then]
    self:Static Mesh Object Reference=Get L_MeshToModify()
    QualityLevelMinimumLODs:Map of EPerQualityLevels Enums to Integers=Get L_Env_LODQuality_Defaults() -> (L_Env_LODQuality_Defaults:Map of EPerQualityLevels Enums to Integers=Low)
    Default:int=-1
   ) -> (then)
   SetLodScreenSizes(
    execute=@n28.else
    self=@n12
    StaticMesh=@n13
    ScreenSizes=Get L_LOD_ScreenSizes() -> (L_LOD_ScreenSizes:"Array of Float (single-precision)s")
   )
   n14 = Get L_MeshToModify()
   SaveLoadedAsset(execute=@n23.then, AssetToSave=@n14, bOnlyIfIsDirty=false)
   AsyncQueryFileState!(execute, FileStateCallback:Delegate, InFile:String, bSilent:bool=false) -> (then)
   Branch(Condition="Equal Exactly (String)"(A=GetPlatformUserName!() -> (ReturnValue:String), B=brianfasten))
   n15 = Get L_MeshToModify()
   Branch(Condition=Break Mesh Nanite Settings(MeshNaniteSettings=GetNaniteSettings(StaticMesh=@n15)) -> (
     bEnabled=~
     bExplicitTangents=~
     bLerpUVs=~
     bSeparable=~
     bVoxelNDF=~
     bVoxelOpacity=~
     KeepPercentTriangles=~
     TrimRelativeError=~
     FallbackPercentTriangles=~
     FallbackRelativeError=~
     MaxEdgeLengthFactor=~
     RayBackUp=~
    ))
   Set members in Mesh Nanite Settings!(execute:Exec, StructRef:"Mesh Nanite Settings Structure (by ref)"=@n11, bEnabled:bool=False) -> (then:Exec, StructOut:"Mesh Nanite Settings Structure (by ref)") {is=title K2Node_SetFieldsInStruct}
   n16 = Get L_MeshToModify()
   Set LODForCollision!(execute=@n26.then, LODForCollision:int=2, self:Static Mesh Object Reference=@n16) -> (then, Output_Get:int=0)
   n17 = GetNumLODs(self=@n16)
   Set LODForCollision(execute=@n26.else, LODForCollision=int - int(A=@n17), self=@n16) -> (Output_Get=0)
   n18 = GetNumLODs(self=Get L_MeshToModify())
   n19 = "To String (Name)"(InName=CreateAssetData!(InAsset=@n6, bAllowBlueprintClass=false) -> (
     ReturnValue_PackageName:Name
     ReturnValue_PackagePath:Name
     ReturnValue_AssetName:Name
     ReturnValue_AssetClassPath:Top Level Asset Path Structure
    ).ReturnValue_PackagePath)
   n20 = Split!(InPath:String=@n19) -> (PathPart:String, FilenamePart:String, ExtensionPart:String)
   n21 = FindSubstring(SearchIn=@n20, Substring=Get L_TextKey() -> (L_TextKey:String))
   n22 = CheckoutLoadedAsset!(execute=@n25.then_4, AssetToCheckout:Object Reference=@n14) -> (then, ReturnValue:bool)
   > n23 = Branch(Condition=@n22)
   n24 = SetLODGroup!(
    execute=@n27.else
    self:Static Mesh Editor Subsystem Object Reference=Get StaticMeshEditorSubsystem()
    StaticMesh:Static Mesh Object Reference=Get L_MeshToModify()
    LODGroup:Name=Get L_Preset_LOD_Group() -> (L_Preset_LOD_Group:Name=None)
    bRebuildImmediately:bool=true
   ) -> (then, ReturnValue:bool)
   n25 = Sequence!(execute=[@n31.then, @n30.then]) -> (then_0, then_1, then_2:Exec, then_3:Exec, then_4:Exec)
   n26 = Branch(execute=@n25.then_2, Condition="integer >= integer"(A=@n17, B=3))
   n27 = Branch(execute=@n25.then_0, Condition="integer > integer"(A=@n18, B=2))
   n28 = Branch(execute=@n25.then_1, Condition=Get ApplyPropScreenSizes() -> (ApplyPropScreenSizes:bool))
   n29 = Branch(execute=@n6.then, Condition="Not Equal (Integer)"(A=@n21, B=-1))
   > n30 = Set ApplyPropScreenSizes(ApplyPropScreenSizes:bool=true) -> (Output_Get:bool)
   n31 = Set ApplyPropScreenSizes(execute=@n29.else, ApplyPropScreenSizes=false)
  ]}
 EventGraph {links [], nodes []}
]
variables [STR_EnvNaniteSettings {
  type=MeshNaniteSettings
  default {
   bEnabled=true
   bExplicitTangents=false
   bLerpUVs=true
   bSeparable=false
   bVoxelNDF=true
   bVoxelOpacity=false
   boneWeightPrecision=-1
   displacementUVChannel=0
   fallbackPercentTriangles=1
   fallbackRelativeError=0
   fallbackTarget=RelativeError
   generateFallback=PlatformDefault
   keepPercentTriangles=1
   maxEdgeLengthFactor=0
   normalPrecision=-1
   numRays=1
   positionPrecision=-2147483648
   rayBackUp=0
   shapePreservation=None
   tangentPrecision=-1
   targetMinimumResidencyInKB=0
   trimRelativeError=0
   voxelLevel=0
   displacementMaps []
   naniteAssemblyData {nodes [], parts []}
  }
 }]
