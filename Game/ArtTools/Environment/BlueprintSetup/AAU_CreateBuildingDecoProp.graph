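# Editor utility that creates BuildingDeco versions of selected Blueprint assets.
# Create_BuildingDeco_Version resolves each NativeParentClass tag and skips SimpleDestructibleProp children.
# Others are duplicated under the same name into a BuildingDeco subfolder, then reparented to BuildingDeco.
# An uncalled helper converts BP_ names to BSG_, and disconnected reparent and duplicate chains remain in the graph.
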EditorUtilityBlueprint /Game/ArtTools/Environment/BlueprintSetup/AAU_CreateBuildingDecoProp
graphs [
 Convert_BP_To_BSG_Name {nodes [
   n1 = Convert_BP_To_BSG_Name!() -> (then, InName:Name) {is=title K2Node_FunctionEntry}
   > Set L_Name(L_Name:String="To String (Name)"(InName=@n1)) -> (Output_Get:String)
   > n2 = Branch(Condition=StartsWith(SourceString=Get L_Name() -> (L_Name:String), InPrefix=BSG_))
   > Return Node!(execute, ReturnValue:String=Get L_Name())
   n3 = Sequence(execute=@n2.else)
   Return Node(execute=@n3.then_1, ReturnValue=Get L_Name())
   > n4 = Branch(Condition=StartsWith(SourceString=Get L_Name(), InPrefix=BP_))
   > Set L_Name(L_Name=Replace(SourceString=Get L_Name(), From=BP_, To=BSG_))
   Set L_Name(execute=@n4.else, L_Name=Append(A=BSG_, B=Get L_Name()))
  ]}
 Create_BuildingDeco_Version {nodes [
   n1 = GetBlueprintAsset()
   > ReparentBlueprint(Blueprint=@n1, NewParentClass=/Game/Environments/Blockout/Blueprints/Props/Blockout_Prop_Small_SDP.Blockout_Prop_Small_SDP_C)
   n2 = GetBlueprintAsset()
   > n3 = ParentClass!(execute, BlueprintObj:Blueprint Object Reference=@n2) -> (then, ReturnValue:Object Class Reference)
   > Branch(Condition="Not Equal (Class)"(
     A:Object Class Reference=@n3
     B:Object Class Reference=&Blockout_Prop_Small_SDP_C
    ) -> (ReturnValue:bool=~) {is=title K2Node_PromotableOperator})
   > n4 = GetPathNameForLoadedAsset(LoadedAsset=@n2)
   > PrintString(InString=@n4)
   > DuplicateAsset!(
    execute:Exec
    self:Object Reference=@n2
    AssetName:String=Append!(A=BSG, B=_, C:String=GetObjectName(Object=@n2)) -> (ReturnValue)
    PackagePath:String=@n4
    OriginalObject:Object Reference=@n2
   ) -> (then:Exec, ReturnValue:Object Reference) {is=title K2Node_Message}
   Create_BuildingDeco_Version() {is=title K2Node_FunctionEntry}
   > n5 = GetSelectedAssets()
   > n6 = Filter by Class!(
    execute
    TargetArray:Array of Object References=@n5
    ObjectClass:Object Class Reference=/Script/Engine.Blueprint
    FilterType:EEditorScriptingFilterType Enum
   ) -> (then, ReturnValue:Array of Blueprint Object References)
   > n7 = For Each Loop(Array:Array of Blueprint Object References=@n6) -> (Array Element:Blueprint Object Reference)
   n8 = CreateAssetData(InAsset=@n7, bAllowBlueprintClass=false)
   n9 = GetTagValue(InAssetData=@n8, InTagName=MakeLiteralName(Value=NativeParentClass))
   > n10 = LoadClassAsset_Blocking(AssetClass=To Soft Class Reference(SoftClassPath=MakeSoftClassPath(PathString=@n9)))
   > n11 = Branch(Condition=ClassIsChildOf(TestClass=@n10, ParentClass=/Script/FortniteGame.SimpleDestructibleProp))
   n12 = Break Asset Data(AssetData=@n8)
   n13 = DuplicateLoadedAsset!(
    execute=@n11.else
    SourceAsset:Object Reference=@n7
    DestinationAssetPath:String=Append(A="To String (Name)"(InName=@n12.PackagePath), B=/BuildingDeco/, C="To String (Name)"(InName=@n12.AssetName))
   ) -> (then, ReturnValue:Object Reference)
   > n14 = GetBlueprintAsset(Object=@n13)
   > ReparentBlueprint(Blueprint=@n14, NewParentClass=/Script/FortniteGame.BuildingDeco)
   n15 = For Each Loop(Array:Array of Building Prop Object References) -> (Array Element:Building Prop Object Reference)
   > n16 = GetPathNameForLoadedAsset(LoadedAsset=@n15)
   > DuplicateAsset(
    self=@n15
    AssetName=Append(A=BSG, B=_, C=GetDisplayName(Object=@n15))
    PackagePath=GetPath!(InPath:String=@n16) -> (ReturnValue:String)
    OriginalObject=@n15
   )
   ReparentBlueprint(NewParentClass=/Script/FortniteGame.BuildingStaticGeo)
   GetDisplayName(Object=@n13)
  ]}
 EventGraph {links [], nodes []}
]
variables []
