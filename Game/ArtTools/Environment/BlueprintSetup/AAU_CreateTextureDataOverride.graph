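# Editor utility that builds BuildingTextureData assets from textures and an optional Material Instance selected in the Content Browser.
# It sorts textures into Diffuse, Normal, Specular and Displacement by matching LOD group and compression settings as strings.
# A TD_ name and TextureData folder path come from the diffuse texture's path, and a template is duplicated there if absent.
# It then assigns textures via SetEditorProperty, sets the override material and saves.
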EditorUtilityBlueprint /Game/ArtTools/Environment/BlueprintSetup/AAU_CreateTextureDataOverride
graphs [
 AssignMaterialOverrideToTextureData {nodes [
   AssignMaterialOverrideToTextureData() {is=title K2Node_FunctionEntry}
   > Branch(Condition=IsValid(Object=Get OverrideMaterial() -> (OverrideMaterial:Material Interface Object Reference)))
   > n1 = Cast To BuildingTextureData!(execute, Object=Get NewTD() -> (NewTD:Object Reference)) -> (then, CastFailed, AsBuilding Texture Data:Building Texture Data Object Reference) {is=title K2Node_DynamicCast}
   > Set OverrideMaterial!(execute, OverrideMaterial:Material Interface Object Reference=Get OverrideMaterial(), self:Building Texture Data Object Reference=@n1) -> (then, Output_Get:Material Interface Object Reference)
  ]}
 AssignTexturesToTextureData {nodes [
   n1 = Cast To BuildingTextureData(Object=Get NewTD())
   > SetEditorProperty(Object=@n1, PropertyName=Diffuse, PropertyValue:Texture 2D Object Reference=Get T_Diffuse() -> (T_Diffuse:Texture 2D Object Reference))
   n2 = Cast To BuildingTextureData(Object=Get NewTD())
   > SetEditorProperty(Object=@n2, PropertyName=Displacement, PropertyValue=Get T_Displacement() -> (T_Displacement:Texture 2D Object Reference))
   AssignTexturesToTextureData() {is=title K2Node_FunctionEntry}
   > n3 = Cast To BuildingTextureData(Object=Get NewTD())
   > n4 = SetEditorProperty(Object=@n3, PropertyName=Diffuse, PropertyValue=Get T_Diffuse())
   > Branch(Condition=@n4)
   > PrintString(InString=DiffuseSet)
   > SetEditorProperty(Object=@n3, PropertyName=Normal, PropertyValue=Get T_Normal() -> (T_Normal:Texture 2D Object Reference))
   > SetEditorProperty(Object=@n3, PropertyName=Specular, PropertyValue=Get T_Spec() -> (T_Spec:Texture 2D Object Reference))
   > SetEditorProperty(
    Object=@n3
    PropertyName=Displacement
    PropertyValue:Texture 2D Soft Object Reference=Make Soft Reference!(Input:Texture 2D Object Reference=Get T_Displacement()) -> (Output:Texture 2D Soft Object Reference)
   )
   > PrintString(InString=HeightMapSet)
  ]}
 CreateNewTextureData {nodes [
   n1 = CreateNewTextureData!() -> (then, SourceTexture:Texture 2D Object Reference) {is=title K2Node_FunctionEntry}
   > n2 = GetPathNameForLoadedAsset(LoadedAsset=@n1)
   > n3 = Set L_SourcePath(L_SourcePath:String=@n2) -> (Output_Get:String)
   n4 = Split(SourceString=@n3, InStr=Textures, SearchDir=FromStart)
   > PrintString(InString=@n4)
   n5 = Split(SourceString=@n4, InStr=/, SearchDir=FromEnd)
   > PrintString(InString=@n5)
   n6 = Split(SourceString=@n5, InStr=/, SearchDir=FromEnd)
   > PrintString(InString=@n6.RightS)
   > Set L_Category(L_Category:String=@n6.RightS) -> (Output_Get:String)
   n7 = Split(
    SourceString=RightChop!(SourceString:String=Split(SourceString=@n4.RightS, InStr=., SearchDir=FromStart), Count:int=1) -> (ReturnValue:String)
    InStr=_
    SearchDir=FromEnd
   )
   > Set TD_Name(TD_Name:String=Append(A=TD_, B=Split(SourceString=@n7, InStr=_, SearchDir=FromEnd))) -> (Output_Get:String)
   n8 = Split(SourceString=Get L_SourcePath() -> (L_SourcePath:String), InStr=/, SearchDir=FromEnd)
   n9 = Get TD_Name() -> (TD_Name:String)
   > Set L_DestinationPath(L_DestinationPath:String=Append!(A=Split(SourceString=@n8, InStr=/, SearchDir=FromEnd), B=/, C:String=TextureData, D:String=/, E:String=@n9, F:String=., G:String=@n9) -> (ReturnValue)) -> (Output_Get:String)
   n10 = Get L_DestinationPath() -> (L_DestinationPath:String)
   > n11 = DoesAssetExist(AssetPath=@n10)
   > n12 = Branch(Condition=@n11)
   n13 = DuplicateAsset!(execute=@n12.else, SourceAssetPath:String=Get L_TemplateAsset() -> (L_TemplateAsset:String), DestinationAssetPath:String=@n10) -> (then, ReturnValue:Object Reference)
   > Set NewTD(NewTD:Object Reference=@n13) -> (Output_Get:Object Reference)
  ]}
 CreateTextureDataAsset {nodes [
   n1 = Filter by Class!(
    execute
    TargetArray:Array of Object References=Get V2_AR_SelectedAssets() -> (V2_AR_SelectedAssets:Array of Object References)
    ObjectClass:Object Class Reference=/Script/Engine.Texture2D
    FilterType:EEditorScriptingFilterType Enum
   ) -> (then, ReturnValue:Array of Texture 2D Object References)
   > Set V2_AR_FilteredTextures(V2_AR_FilteredTextures:Array of Texture 2D Object References=@n1) -> (Output_Get:Array of Texture 2D Object References)
   n2 = GetSelectedAssets()
   > Set V2_AR_SelectedAssets(V2_AR_SelectedAssets:Array of Object References=@n2) -> (Output_Get:Array of Object References)
   n3 = Cast To BuildingTextureData(Object=Get NewTD())
   > Set OverrideMaterial(OverrideMaterial=/Hermes/Sets/AlleyTown/Materials/MI_AlleyTown_1layer, self=@n3)
   n4 = Cast To BuildingTextureData(Object=Get NewTD())
   > n5 = Cast To MaterialInstanceConstant(Object=Get MatObject() -> (MatObject:Object Reference))
   > Set OverrideMaterial(OverrideMaterial=@n5, self=@n4)
   CreateTextureDataAsset() {is=title K2Node_FunctionEntry}
   > n6 = GetSelectedAssets()
   > n7 = Sequence()
   > n8 = Filter by Class(TargetArray=@n6, ObjectClass=Get ObjectClass() -> (ObjectClass:Object Class Reference)) -> (ReturnValue:Array of Object References)
   > Set AR_TexObj(AR_TexObj:Array of Object References=@n8) -> (Output_Get:Array of Object References)
   n9 = Filter by Class(execute=@n7.then_1, TargetArray=@n6, ObjectClass=/Script/Engine.MaterialInstanceConstant) -> (ReturnValue:Array of Material Instance Constant Object References)
   > Set OverrideMaterial!(
    execute
    OverrideMaterial="Get (a copy)"(Array:Array of Material Instance Constant Object References=@n9) -> (Output:Material Instance Constant Object Reference)
   ) -> (then, Output_Get)
   > n10 = For Each Loop(Array:Array of Object References=Get AR_TexObj() -> (AR_TexObj:Array of Object References)) -> (Array Element:Object Reference)
   > n11 = Cast To Texture2D(Object=@n10)
   > Set TextureToIdenfity(TextureToIdenfity:Texture 2D Object Reference=@n11) -> (Output_Get:Texture 2D Object Reference)
   > n12 = Sequence!(execute) -> (then_0, then_1, then_2:Exec)
   > SortTexturesByGroup()
   CreateNewTextureData!(execute=@n12.then_1, self, SourceTexture:Texture 2D Object Reference=Get T_Diffuse()) -> (then) {is=title K2Node_CallFunction}
   > AssignTexturesToTextureData() {is=title K2Node_CallFunction}
   > AssignMaterialOverrideToTextureData() {is=title K2Node_CallFunction}
   SaveLoadedAsset(execute=@n12.then_2, AssetToSave=Get NewTD(), bOnlyIfIsDirty=true)
   n13 = For Each Loop(Array:Array of Texture 2D Object References=Make Array!(
     "[0]":Texture 2D Object Reference=Get T_Diffuse()
     "[1]":Texture 2D Object Reference=Get T_Normal()
     "[2]":Texture 2D Object Reference=Get T_Spec()
     "[3]":Texture 2D Object Reference=Get T_Displacement()
    ) -> (Array:Array of Texture 2D Object References)) -> (Array Element:Texture 2D Object Reference)
   > PrintString(InString=GetDisplayName(Object=@n13))
   n14 = For Each Loop(Array:Array of Object References) -> (Array Element:Object Reference)
   n15 = Get Class(Object=@n14) -> (ReturnValue:Object Class Reference)
   n16 = Get ObjectClass()
   > Branch(Condition=MatchesWildcard(SourceString=Get Class Display Name(Class=@n16), Wildcard=Get Class Display Name(Class=@n15)))
   > Add(
    TargetArray:Array of Object References=Get AR_Textures() -> (AR_Textures:Array of Object References)
    NewItem:"Object Reference (by ref)"=@n14
   ) {is=title K2Node_CallArrayFunction}
   Reroute Node(InputPin=@n14.Completed)
   n17 = For Each Loop(Array=Get V2_AR_FilteredTextures() -> (V2_AR_FilteredTextures:Array of Texture 2D Object References))
   > Add Unique(
    TargetArray=Get V2_AR_UniqueNames() -> (V2_AR_UniqueNames:Array of Strings)
    NewItem=LeftChop(SourceString=GetDisplayName(Object=@n17), Count=1)
   )
   n18 = For Each Loop(Array:Array of Strings=Get V2_AR_UniqueNames()) -> (Array Element:String)
   Add Unique(
    TargetArray=Get V2_AR_TD_Names() -> (V2_AR_TD_Names:Array of Strings)
    NewItem=Append!(A=TD_, B=RightChop(SourceString=@n18, Count=1)) -> (ReturnValue)
   )
   SaveLoadedAsset(AssetToSave=Get NewTD(), bOnlyIfIsDirty=true)
   n19 = Get MaterialClass() -> (MaterialClass:Object Class Reference)
   n20 = Reroute Node(InputPin:Object Reference) -> (OutputPin:Object Reference)
   n21 = Get Class(Object=@n20)
   Branch(Condition=MatchesWildcard(SourceString=Get Class Display Name(Class=@n19), Wildcard=Get Class Display Name(Class=@n21)))
   Set MatObject(MatObject:Object Reference=@n20) -> (Output_Get:Object Reference)
  ]}
 EventGraph {links [], nodes []}
 IdentifyTextureGroup {nodes [
   n1 = IdentifyTextureGroup!() -> (then, InputTexture:Texture Object Reference, TextureGroup:String, TextureCompression:String) {is=title K2Node_FunctionEntry}
   n2 = MatchesWildcard(
    SourceString=Enum to String!(Enumerator:TextureGroup Enum=Get LODGroup!(self:Texture Object Reference=@n1) -> (LODGroup:TextureGroup Enum)) -> (ReturnValue:String)
    Wildcard=@n1.TextureGroup
   )
   > Branch(Condition=@n2)
   n3 = Get CompressionSettings(self=@n1)
   > Branch(Condition=MatchesWildcard(SourceString=Enum to String(Enumerator:TextureCompressionSettings Enum=@n3), Wildcard=@n1.TextureCompression))
   > Return Node!(execute, PositiveMatch:bool=@n2)
  ]}
 SortTexturesByGroup {nodes [
   SortTexturesByGroup() {is=title K2Node_FunctionEntry}
   > n1 = IdentifyTextureGroup!(
    execute
    self
    InputTexture:Texture Object Reference=Get TextureToIdenfity() -> (TextureToIdenfity:Texture 2D Object Reference)
    TextureGroup:String=World
    TextureCompression:String="Default (BC1 or BC3 with A)"
   ) -> (then, PositiveMatch:bool) {is=title K2Node_CallFunction}
   > n2 = Branch(Condition=@n1)
   > Set T_Diffuse(T_Diffuse:Texture 2D Object Reference=Get TextureToIdenfity()) -> (Output_Get:Texture 2D Object Reference)
   n3 = IdentifyTextureGroup(execute=@n2.else, InputTexture=Get TextureToIdenfity(), TextureGroup=WorldNormalMap, TextureCompression="Normalmap (BC5)")
   > n4 = Branch(Condition=@n3)
   > Set T_Normal(T_Normal:Texture 2D Object Reference=Get TextureToIdenfity()) -> (Output_Get:Texture 2D Object Reference)
   n5 = IdentifyTextureGroup(execute=@n4.else, InputTexture=Get TextureToIdenfity(), TextureGroup=WorldSpecular, TextureCompression="Masks (no sRGB)")
   > n6 = Branch(Condition=@n5)
   > Set T_Spec(T_Spec:Texture 2D Object Reference=Get TextureToIdenfity()) -> (Output_Get:Texture 2D Object Reference)
   IdentifyTextureGroup(execute=@n6.else, InputTexture=Get TextureToIdenfity(), TextureGroup=World, TextureCompression="Grayscale (G8/16, RGB8 sRGB) from source R")
   > Set T_Displacement(T_Displacement:Texture 2D Object Reference=Get TextureToIdenfity()) -> (Output_Get:Texture 2D Object Reference)
  ]}
]
variables [
 AR_TexObj {type="array<Object>", default []}
 AR_Texture2D {type="array<Object>", default []}
 AR_Textures {type="array<Object>", default []}
 MaterialClass {type=Class, default=&MaterialInstanceConstant}
 MatObject {type=Object, default=null}
 NewTD {type=Object, default=null}
 ObjectClass {type=Class, default=&Texture2D}
 OverrideMaterial {type=MaterialInterface, default=null}
 T_Diffuse {type=Texture2D}
 T_Displacement {type=Texture2D}
 T_Normal {type=Texture2D}
 T_Spec {type=Texture2D}
 TD_Name {type=string, default=""}
 TextureGroup {type=string, default=World}
 TextureToIdenfity {type=Texture2D}
 V2_AR_FilteredTextures {type="array<Texture2D>", default []}
 V2_AR_SelectedAssets {type="array<Object>", default []}
 V2_AR_TD_Names {type="array<string>", default []}
 V2_AR_UniqueNames {type="array<string>", default []}
]
