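# Editor utility that batch-creates glass Simple Destructible Prop component Blueprints from selected static meshes.
# For each mesh, it derives a name without the SM_ prefix and a Blueprints/Components/Glass path, then duplicates BPC_SDP_GlassProvider unless that asset exists.
# It sets StaticMesh on the new Blueprint's class default object, then batch-saves.
# Map-based creation, CreateBlueprintComponent and UpdateComponentSettings (tags, collision, resource type) are unconnected to the main flow.
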EditorUtilityBlueprint /Game/ArtTools/Environment/BlueprintSetup/AAU_Create_SDP_GlassComponent
graphs [
 AssignMeshToComponent {nodes [
   AssignMeshToComponent() {is=title K2Node_FunctionEntry}
   > n1 = GetBlueprintAsset(Object=Get SDP_Component() -> (SDP_Component:Object Reference))
   > Is Valid(InputObject=@n1)
   > n2 = GeneratedClass!(execute, BlueprintObj:Blueprint Object Reference=@n1) -> (then, ReturnValue:Object Class Reference)
   > n3 = GetClassDefaultObject!(execute, Class:Object Class Reference=@n2) -> (then, ReturnValue:Object Reference)
   > n4 = Cast To StaticMeshComponent(Object=@n3)
   > SetEditorProperty(Object=@n4, PropertyName=StaticMesh, PropertyValue:Static Mesh Object Reference=Get GlassMesh() -> (GlassMesh:Static Mesh Object Reference))
  ]}
 BuildRequiredAssetData {nodes [
   n1 = BuildRequiredAssetData!() -> (then, PackageName:Name) {is=title K2Node_FunctionEntry}
   n2 = Split!(InPath:String="To String (Name)"(InName=@n1)) -> (PathPart:String, FilenamePart:String, ExtensionPart:String)
   n3 = Split!(SourceString:String=@n2.FilenamePart, InStr:String=SM_, SearchCase:ESearchCase Enum, SearchDir:ESearchDir Enum=FromStart) -> (LeftS:String, RightS:String, ReturnValue:bool)
   > n4 = Branch(Condition=@n3.ReturnValue)
   n5 = Append!(
    A=Split(SourceString=@n2, InStr=/, SearchDir=FromEnd)
    B=/
    C:String=Blueprints
    D:String=/
    E:String=Components
    F:String=/
    G:String=Glass
    H:String=/
   ) -> (ReturnValue)
   > Return Node!(execute, NewAssetName:String=Append!(A=BP_SDP_Glass, B=_, C=@n3.RightS) -> (ReturnValue), NewAssetPath:String=@n5)
   Return Node(execute=@n4.else, NewAssetName=Append(A=BP_SDP_Glass, B=_, C=@n2.FilenamePart), NewAssetPath=@n5)
  ]}
 CreateBlueprintComponent {nodes [
   n1 = GetAssetTools()
   > CreateAsset(self=@n1, AssetClass=/Script/FortniteGame.SimpleDestructiblePropProviderComponent)
   n2 = CreateBlueprintComponent!() -> (then, AssetName:String, PackagePath:String) {is=title K2Node_FunctionEntry}
   > n3 = CreateBlueprintAssetWithParent!(
    execute
    AssetPath:String=Append!(A=@n2.PackagePath, B=@n2) -> (ReturnValue)
    ParentClass:Object Class Reference=/Game/ArtTools/Environment/BlueprintSetup/Templates/BPC_SDP_GlassProvider.BPC_SDP_GlassProvider_C
   ) -> (then, ReturnValue:Blueprint Object Reference)
   > n4 = Set SDP_Component(SDP_Component:Object Reference=@n3) -> (Output_Get:Object Reference)
   > SaveLoadedAsset(AssetToSave=@n4, bOnlyIfIsDirty=true)
   CreateBlueprintAssetWithParent(ParentClass=&SimpleDestructiblePropProviderComponent)
  ]}
 CreateGlassComponents {nodes [
   Clear!(
    execute
    TargetMap:Map of Integers to Static Mesh Object References=Get Map_CreateOrder() -> (Map_CreateOrder:Map of Integers to Static Mesh Object References)
   ) -> (then) {is=title K2Node_CallFunction}
   > Clear(TargetMap:Map of Static Mesh Object References to Object References=Get Map_MeshToComponentPair() -> (Map_MeshToComponentPair:Map of Static Mesh Object References to Object References))
   > Clear(TargetArray:Array of Static Mesh Object References=Get AR_SelectedMeshes() -> (AR_SelectedMeshes:Array of Static Mesh Object References)) {is=title K2Node_CallArrayFunction}
   > FilterSelectedAssets()
   > n1 = For Each Loop(Array:Array of Static Mesh Object References=Get AR_SelectedMeshes()) -> (Array Element:Static Mesh Object Reference)
   n2 = "For Each Loop (Map)"!(execute:Exec=@n1.Completed, MapPin:Map of Integers to Static Mesh Object References=Get Map_CreateOrder(), BreakPin:Exec) -> (then:Exec, KeyPin:int, ValuePin:Static Mesh Object Reference, CompletedPin:Exec)
   > n3 = ExtractDataFromMesh!(execute, self, MeshToProcess:Static Mesh Object Reference=@n2.ValuePin) -> (then, PackageName:Name=None, MeshObjectRef:Static Mesh Object Reference)
   > n4 = BuildRequiredAssetData!(execute, self, PackageName:Name=@n3) -> (then, NewAssetName:String, NewAssetPath:String) {is=title K2Node_CallFunction}
   > n5 = DuplicateAsset!(
    execute
    SourceAssetPath:String=Get ComponentTemplatePath() -> (ComponentTemplatePath:String)
    DestinationAssetPath:String=Append(A=@n4.NewAssetPath, B=@n4)
   ) -> (then, ReturnValue:Object Reference)
   > Add!(
    execute
    TargetMap:Map of Static Mesh Object References to Object References=Get Map_MeshToComponentPair()
    Key:"Static Mesh Object Reference (by ref)"=@n3.MeshObjectRef
    Value:"Object Reference (by ref)"=@n5
   ) -> (then) {is=title K2Node_CallFunction}
   n6 = "For Each Loop (Map)"(execute=@n2.CompletedPin, MapPin:Map of Static Mesh Object References to Object References=Get Map_MeshToComponentPair()) -> (KeyPin:Static Mesh Object Reference, ValuePin:Object Reference)
   > n7 = Cast To BPC_SDP_GlassProvider!(execute, Object=@n6.ValuePin) -> (then, CastFailed, AsBPC SDP Glass Provider:BPC SDP Glass Provider Object Reference) {is=title K2Node_DynamicCast}
   > n8 = SetStaticMesh(self=@n7, NewMesh=@n6)
   > n9 = Branch(Condition=@n8)
   > PrintString(InString=MeshSet)
   PrintString(execute=@n9.else, InString=MeshNotSet)
   PrintString(execute=@n7.CastFailed, InString=CastFailed)
   Add(
    execute=@n1.LoopBody
    TargetMap:Map of Integers to Static Mesh Object References=Get Map_CreateOrder()
    Key:"Integer (by ref)"=@n1.Array Index
    Value:"Static Mesh Object Reference (by ref)"=@n1
   )
   CreateBlueprintComponent!(execute, self, AssetName:String, PackagePath:String) -> (then) {is=title K2Node_CallFunction}
   n10 = Get SDP_Component()
   > n11 = Cast To SimpleDestructiblePropProviderComponent!(execute, Object=@n10) -> (then, CastFailed, AsSimple Destructible Prop Provider Component:Simple Destructible Prop Provider Component Object Reference) {is=title K2Node_DynamicCast}
   n12 = SetStaticMesh(execute=[@n15.then, @n11.then], self=@n11, NewMesh=Get GlassMesh())
   > n13 = Branch(Condition=@n12)
   > PrintString(InString=MeshUpdateSuccessful)
   PrintString(execute=@n13.else, InString=MeshUpdateFailed)
   PrintString(execute=[@n15.CastFailed, @n11.CastFailed], InString=CastToSDPComponentFailed)
   n14 = GetMetadataTag!(execute, Object:Object Reference, Tag:Name=SDP) -> (then, ReturnValue:String)
   > PrintString(InString=Append(A=FoundTagValue, B=@n14))
   n15 = Cast To StaticMeshComponent(Object=@n10)
   n16 = Cast To StaticMeshComponent(Object=Get SDP_Component())
   > n17 = SetStaticMesh(self=@n16, NewMesh=Get GlassMesh())
   > Branch(Condition=@n17)
   > Add Unique(TargetArray:Array of Object References, NewItem:"Object Reference (by ref)"=Get SDP_Component())
   n18 = Cast To BPC_SDP_GlassProvider(Object=Get SDP_Component())
   > n19 = SetStaticMesh(self=@n18)
   > n20 = Branch(Condition=@n19)
   > PrintString(InString=MeshSet)
   PrintString(execute=@n20.else, InString=MeshNotSet)
   PrintString(execute=@n18.CastFailed, InString=CastFailed)
   CreateGlassComponents() {is=title K2Node_FunctionEntry}
   > Clear(TargetArray=Get AR_SelectedMeshes())
   > Clear(TargetArray:Array of Object References=Get L_AssetsToSave() -> (L_AssetsToSave:Array of Object References))
   > FilterSelectedAssets()
   > n21 = For Each Loop(Array=Get AR_SelectedMeshes())
   SaveLoadedAssets(execute=@n21.Completed, AssetsToSave=Get L_AssetsToSave(), bOnlyIfIsDirty=true)
   n22 = ExtractDataFromMesh(execute=@n21.LoopBody, MeshToProcess=@n21) -> (PackageName=None)
   > Set GlassMesh(GlassMesh:Static Mesh Object Reference=@n22.MeshObjectRef) -> (Output_Get:Static Mesh Object Reference)
   > n23 = BuildRequiredAssetData(PackageName=@n22)
   n24 = Append(A=@n23.NewAssetPath, B=@n23)
   > n25 = DoesAssetExist(AssetPath=@n24)
   > n26 = Branch(Condition=@n25)
   n27 = DuplicateAsset(execute=@n26.else, SourceAssetPath=Get ComponentTemplatePath(), DestinationAssetPath=@n24)
   > Set SDP_Component(SDP_Component=@n27)
   > AssignMeshToComponent() {is=title K2Node_CallFunction}
   > Add Unique(TargetArray=Get L_AssetsToSave(), NewItem=Get SDP_Component())
   BreakGameplayTagContainer!(GameplayTagContainer:"Gameplay Tag Container Structure (by ref)"=Get GameplayTags!(self:Simple Destructible Prop Provider Component Object Reference) -> (GameplayTags:Gameplay Tag Container Structure)) -> (GameplayTags:Array of Gameplay Tag Structures)
   SetMetadataTag!(execute, Object:Object Reference=Get SDP_Component(), Tag:Name=SDP, Value:String=Glass) -> (then)
   UpdateComponentSettings()
   UpdateComponentSettings()
   SaveLoadedAssets(bOnlyIfIsDirty=true)
   Cast To SimpleDestructiblePropProviderComponent(Object=Get SDP_Component())
  ]}
 EventGraph {links [], nodes []}
 ExtractDataFromMesh {nodes [
   n1 = GetSubPaths!(execute, self:Asset Registry Interface=GetAssetRegistry(), InBasePath:String, bInRecurse:bool=false) -> (then, OutPathList:Array of Strings)
   > For Each Loop(Array:Array of Strings=@n1) -> (Array Element:String)
   n2 = ExtractDataFromMesh!() -> (then, MeshToProcess:Static Mesh Object Reference) {is=title K2Node_FunctionEntry}
   > n3 = Set L_MeshToProcess(L_MeshToProcess:Static Mesh Object Reference=@n2) -> (Output_Get:Static Mesh Object Reference)
   > n4 = Set L_PackageName(L_PackageName:Name=Break Asset Data(AssetData=CreateAssetData(InAsset=@n3, bAllowBlueprintClass=false))) -> (Output_Get:Name=None)
   > Return Node!(
    execute
    PackageName:Name=@n4
    MeshObjectRef:Static Mesh Object Reference=Get L_MeshToProcess() -> (L_MeshToProcess:Static Mesh Object Reference)
   )
   Get Object Path String()
   GetSoftObjectPath()
   PrintString(InString=Hello)
   Get AssetEditorSubsystem()
  ]}
 FilterSelectedAssets {nodes [
   FilterSelectedAssets() {is=title K2Node_FunctionEntry}
   > n1 = GetSelectedAssets()
   > n2 = Filter by Class!(
    execute
    TargetArray:Array of Object References=@n1
    ObjectClass:Object Class Reference=&StaticMesh
    FilterType:EEditorScriptingFilterType Enum
   ) -> (then, ReturnValue:Array of Static Mesh Object References)
   > Set AR_SelectedMeshes(AR_SelectedMeshes:Array of Static Mesh Object References=@n2) -> (Output_Get:Array of Static Mesh Object References)
  ]}
 UpdateComponentSettings {nodes [
   UpdateComponentSettings() {is=title K2Node_FunctionEntry}
   > n1 = Cast To SimpleDestructiblePropProviderComponent(Object=Get SDP_Component())
   > n2 = Set L_SDP_To_Modify(L_SDP_To_Modify:Simple Destructible Prop Provider Component Object Reference=@n1) -> (Output_Get:Simple Destructible Prop Provider Component Object Reference)
   > Set GameplayTags!(
    execute
    GameplayTags:Gameplay Tag Container Structure=MakeGameplayTagContainerFromArray!(GameplayTags:Array of Gameplay Tag Structures=Get GlassGameplayTagContainer() -> (GlassGameplayTagContainer:Array of Gameplay Tag Structures)) -> (ReturnValue:Gameplay Tag Container Structure)
    self:Simple Destructible Prop Provider Component Object Reference=@n2
   ) -> (then, Output_Get:Gameplay Tag Container Structure)
   n3 = Get L_SDP_To_Modify() -> (L_SDP_To_Modify:Simple Destructible Prop Provider Component Object Reference)
   > SetMobility!(execute, self:Scene Component Object Reference=@n3, NewMobility:EComponentMobility Enum=Static) -> (then)
   > SetStaticMesh(self=@n3, NewMesh=Get GlassMesh())
   > SetCollisionProfileName!(
    execute
    self:Primitive Component Object Reference=@n3
    InCollisionProfileName:Name=FortBuildingMeshPhysicsOverlapDmg
    bUpdateOverlaps:bool=true
   ) -> (then)
   > Set DestructionResourceType!(
    execute
    DestructionResourceType:EDynamicDestructionResourceType Enum=Glass
    self:Simple Destructible Prop Provider Component Object Reference=@n3
   ) -> (then, Output_Get:EDynamicDestructionResourceType Enum=Invalid)
  ]}
]
variables [
 AR_SelectedMeshes {type="array<StaticMesh>", default []}
 ComponentTemplatePath {type=string, default=/Game/ArtTools/Environment/BlueprintSetup/Templates/BPC_SDP_GlassProvider}
 GlassGameplayTagContainer {type="array<GameplayTag>", default [{tagName=Building.Type.Prop.Glass}, {tagName=Building.Type.Prop.BreaksOnPawnImpact}]}
 GlassMesh {type=StaticMesh, default=null}
 Map_CreateOrder {}
 Map_MeshToComponentPair {}
 ReferenceComponent {type=Object, default=&BPC_SDP_GlassProvider}
 SDP_Component {type=Object, default=null}
 TEST {type=string, default=USimpleDestructiblePropProviderComponent}
]
