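# Editor utility whose EnableNanite function turns on Nanite for the static meshes among the selected assets.
# It loops over them, applying the shared STR_EnvNaniteSettings struct through the Static Mesh Editor Subsystem, which replaces each mesh's Nanite settings.
# After applying, it checks out and saves each asset, or prints the last error message if checkout fails.
# A Set members node that would disable Nanite in that struct is left unconnected.
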EditorUtilityBlueprint /Game/ArtTools/Environment/BlueprintSetup/AAU_EnableNanite
graphs [
 EnableNanite {nodes [
   EnableNanite() {is=title K2Node_FunctionEntry}
   > n1 = GetSelectedAssets()
   > n2 = Filter by Class!(
    execute
    TargetArray:Array of Object References=@n1
    ObjectClass:Object Class Reference=&StaticMesh
    FilterType:EEditorScriptingFilterType Enum
   ) -> (then, ReturnValue:Array of Static Mesh Object References)
   > n3 = For Each Loop(Array:Array of Static Mesh Object References=@n2) -> (Array Element:Static Mesh Object Reference)
   > Set L_MeshToModify(L_MeshToModify:Static Mesh Object Reference=@n3) -> (Output_Get:Static Mesh Object Reference)
   > n4 = Sequence()
   n5 = Get L_MeshToModify() -> (L_MeshToModify:Static Mesh Object Reference)
   n6 = CheckoutLoadedAsset!(execute=@n4.then_1, AssetToCheckout:Object Reference=@n5) -> (then, ReturnValue:bool)
   > n7 = Branch(Condition=@n6)
   > SaveLoadedAsset(AssetToSave=@n5, bOnlyIfIsDirty=false)
   PrintString(execute=@n7.else, InString="To String (Text)"(InText=LastErrorMsg!() -> (ReturnValue:Text)))
   n8 = Get STR_EnvNaniteSettings() -> (STR_EnvNaniteSettings:Mesh Nanite Settings Structure)
   SetNaniteSettings!(
    execute=@n4.then_0
    self:Static Mesh Editor Subsystem Object Reference=Get StaticMeshEditorSubsystem()
    StaticMesh:Static Mesh Object Reference=Get L_MeshToModify()
    NaniteSettings:Mesh Nanite Settings Structure=@n8
    bApplyChanges:bool=true
   ) -> (then)
   Set members in Mesh Nanite Settings!(execute:Exec, StructRef:"Mesh Nanite Settings Structure (by ref)"=@n8, bEnabled:bool=False) -> (then:Exec, StructOut:"Mesh Nanite Settings Structure (by ref)") {is=title K2Node_SetFieldsInStruct}
  ]}
 EventGraph {links [], nodes []}
]
variables [STR_EnvNaniteSettings {
  type=MeshNaniteSettings
  default {
   bEnabled=true
   bExplicitTangents=false
   bLerpUVs=true
   bSeparable=false
   bVoxelNDF=true
   bVoxelOpacity=false
   boneWeightPrecision=-1
   displacementUVChannel=0
   fallbackPercentTriangles=1
   fallbackRelativeError=0
   fallbackTarget=RelativeError
   generateFallback=PlatformDefault
   keepPercentTriangles=1
   maxEdgeLengthFactor=0
   normalPrecision=-1
   numRays=1
   positionPrecision=-2147483648
   rayBackUp=0
   shapePreservation=None
   tangentPrecision=-1
   targetMinimumResidencyInKB=0
   trimRelativeError=0
   voxelLevel=0
   displacementMaps []
   naniteAssemblyData {nodes [], parts []}
  }
 }]
