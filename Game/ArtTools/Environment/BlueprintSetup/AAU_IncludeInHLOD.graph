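# Editor utility that includes selected BuildingProp Blueprint assets in HLOD generation.
# IncludeActorInHLOD filters selected assets by Asset Registry ancestor class, then enables bEnableAutoLODGeneration on each class default object.
# It walks each Blueprint's subobject data, and EnableHlodOnComponent sets that flag on static mesh components and clears their HLOD0 exclusion.
# Dirty assets are saved afterwards; disconnected leftover nodes from other approaches remain in the graph.
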EditorUtilityBlueprint /Game/ArtTools/Environment/BlueprintSetup/AAU_IncludeInHLOD
graphs [
 EnableHlodOnComponent {nodes [
   n1 = EnableHlodOnComponent!() -> (then, Component:Primitive Component Object Reference) {is=title K2Node_FunctionEntry}
   > SetEditorProperty(Object=@n1, PropertyName=bEnableAutoLODGeneration, PropertyValue=MakeLiteralBool(Value=true))
   > Set Excluded From HLOD Level!(execute, self:Primitive Component Object Reference=@n1, HLODLevel:EHLODLevelExclusion Enum=HLOD0, bExcluded:bool=false) -> (then) {is=title K2Node_CallFunction}
  ]}
 EventGraph {links [], nodes []}
 IncludeActorInHLOD {nodes [
   n1 = GetSelectedLevelActors(self=Get EditorActorSubsystem())
   > n2 = For Each Loop(Array:Array of Actor Object References=@n1) -> (Array Element:Actor Object Reference)
   n3 = GetSelectedAssets()
   > n4 = Filter by Class!(
    execute
    TargetArray:Array of Object References=@n3
    ObjectClass:Object Class Reference=/Script/FortniteGame.BuildingActor
    FilterType:EEditorScriptingFilterType Enum
   ) -> (then, ReturnValue:Array of Building Actor Object References)
   > PrintString(InString="To String (Integer)"(InInt=Length(TargetArray:Array of Building Actor Object References=@n4)))
   n5 = Cast To StaticMeshComponent(Object:Wildcard)
   > Set Excluded From HLOD Level(self=@n5, HLODLevel=HLOD0, bExcluded=false)
   PrintString(execute=@n5.CastFailed, InString=Hello)
   IncludeActorInHLOD() {is=title K2Node_FunctionEntry}
   > Clear(TargetArray:Array of Object References=Get L_AR_FilteredAssets() -> (L_AR_FilteredAssets:Array of Object References))
   > n6 = Sequence()
   > n7 = GetSelectedAssetData()
   > n8 = For Each Loop(Array:Array of Asset Data Structures=@n7) -> (Array Element:Asset Data Structure)
   > n9 = Set L_AssetData(L_AssetData:Asset Data Structure=@n8) -> (Output_Get:Asset Data Structure)
   n10 = Break Asset Data(AssetData=@n9)
   n11 = "To String (Name)"(InName=@n10.AssetName)
   > n12 = GetAncestorClassNames!(
    execute
    self:Asset Registry Interface=GetAssetRegistry()
    ClassPathName:Top Level Asset Path Structure=MakeTopLevelAssetPath!(PackageName:String="To String (Name)"(InName=@n10), AssetName:String=Append(A=@n11, B=_C)) -> (ReturnValue:Top Level Asset Path Structure)
   ) -> (then, OutAncestorClassNames:Array of Top Level Asset Path Structures, ReturnValue:bool)
   > Branch(Condition=Contains Item(
     TargetArray:Array of Top Level Asset Path Structures=@n12
     ItemToFind:"Top Level Asset Path Structure (by ref)"=MakeTopLevelAssetPath(PackageName=/Script/FortniteGame, AssetName=BuildingProp)
    ))
   > Add(
    TargetArray:Array of Object References=Get L_AR_FilteredAssets()
    NewItem:"Object Reference (by ref)"=GetAsset(InAssetData=Get L_AssetData() -> (L_AssetData:Asset Data Structure))
   ) {is=title K2Node_CallArrayFunction}
   n13 = Get L_AR_FilteredAssets()
   PrintString(execute=@n6.then_1, InString="To String (Integer)"(InInt=Length(TargetArray:Array of Object References=@n13)))
   n14 = Get L_AR_FilteredAssets()
   > n15 = For Each Loop(Array:Array of Object References=@n14) -> (Array Element:Object Reference)
   SaveLoadedAssets(execute=@n15.Completed, AssetsToSave=@n14, bOnlyIfIsDirty=true)
   n16 = Sequence!(execute=@n15.LoopBody) -> (then_0, then_1, then_2:Exec)
   n17 = Break Asset Data(AssetData=CreateAssetData(InAsset=@n15, bAllowBlueprintClass=false))
   > n18 = GetClassDefaultObjectFromPath!(
    execute
    ClassAssetPath:String=Append!(A="To String (Name)"(InName=@n17), B=., C:String="To String (Name)"(InName=@n17.AssetName), D:String=_C) -> (ReturnValue)
   ) -> (then, ReturnValue:Object Reference)
   > SetEditorProperty(Object=@n18, PropertyName=bEnableAutoLODGeneration, PropertyValue=MakeLiteralBool(Value=true))
   n19 = GetBlueprintAsset(execute=@n16.then_1, Object=@n15)
   > n20 = Gather Subobject Data For Blueprint(self=Get SubobjectDataSubsystem(), Context=@n19)
   > n21 = For Each Loop(Array:Array of Subobject Data Handle Structures=@n20) -> (Array Element:Subobject Data Handle Structure)
   > n22 = GetObjectForBlueprint!(execute, Data:"Subobject Data Structure (by ref)"=GetData(DataHandle=@n21), Blueprint:Blueprint Object Reference=@n19) -> (then, ReturnValue:Object Reference)
   > n23 = Cast To StaticMeshComponent(Object:Object Reference=@n22)
   > EnableHlodOnComponent!(execute, self, Component:Primitive Component Object Reference=@n23) -> (then) {is=title K2Node_CallFunction}
   n24 = For Each Loop(Array:Array of Building Prop Object References) -> (Array Element:Building Prop Object Reference)
   > n25 = Set L_AssetToModify(L_AssetToModify:Actor Object Reference=@n24) -> (Output_Get:Actor Object Reference)
   > n26 = For Each Loop(Array:Array of Static Mesh Component Object References=Get Components By Class!(self:Actor Object Reference=@n25, ComponentClass:Actor Component Class Reference=&StaticMeshComponent) -> (ReturnValue:Array of Static Mesh Component Object References)) -> (Array Element:Static Mesh Component Object Reference)
   > CheckoutLoadedAsset!(execute, AssetToCheckout:Object Reference=Get L_AssetToModify() -> (L_AssetToModify:Actor Object Reference)) -> (then, ReturnValue:bool)
   > Set Excluded From HLOD Level(self=@n26, HLODLevel=HLOD0, bExcluded=false)
   PrintString(InString="To String (Integer)"(InInt=Length(TargetArray:Array of Wildcards)))
   Get Components By Class(self=@n2, ComponentClass=/Script/Engine.ActorComponent) -> (ReturnValue:Array of Actor Component Object References)
   Branch(Condition=Is Excluded From HLOD Level!(self:Primitive Component Object Reference=@n26, HLODLevel:EHLODLevelExclusion Enum=HLOD0) -> (ReturnValue:bool))
   GetAssetsByClass!(execute:Exec, self:Object Reference=Get SubobjectDataSubsystem(), ClassPathName:Top Level Asset Path Structure, bSearchSubClasses:bool=false) -> (then:Exec, OutAssetData:Array of Asset Data Structures, ReturnValue:bool) {is=title K2Node_Message}
  ]}
]
variables []
