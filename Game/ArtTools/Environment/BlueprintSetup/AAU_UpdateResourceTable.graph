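# Editor utility that repoints a curve table row handle on selected BuildingActor Blueprints to the CurveTable variable, AthenaResourceRates by default.
# UpdateResourceTable reads each Blueprint's class default object through GetEditorProperty, using a property name held in a local variable.
# Where the table differs, it checks out the asset, writes the new table with the same row name, then saves.
# A pass-through Python script retypes the loaded object as a CurveTable.
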EditorUtilityBlueprint /Game/ArtTools/Environment/BlueprintSetup/AAU_UpdateResourceTable
graphs [
 EventGraph {links [], nodes []}
 UpdateResourceTable {nodes [
   UpdateResourceTable() {is=title K2Node_FunctionEntry}
   > n1 = GetSelectedAssets()
   > n2 = For Each Loop(Array:Array of Object References=@n1) -> (Array Element:Object Reference)
   > n3 = Set L_Asset(L_Asset:Object Reference=@n2) -> (Output_Get:Object Reference)
   > n4 = GetBlueprintAsset(Object=@n3)
   > n5 = GeneratedClass!(execute, BlueprintObj:Blueprint Object Reference=@n4) -> (then, ReturnValue:Object Class Reference)
   > n6 = GetClassDefaultObject!(execute, Class:Object Class Reference=@n5) -> (then, ReturnValue:Object Reference)
   > n7 = Cast To BuildingActor!(execute, Object=@n6) -> (then, CastFailed, AsBuilding Actor:Building Actor Object Reference) {is=title K2Node_DynamicCast}
   > n8 = Set L_BuildingActorToModify(L_BuildingActorToModify:Building Actor Object Reference=@n7) -> (Output_Get:Building Actor Object Reference)
   > n9 = GetEditorProperty!(execute:Exec, Object:Object Reference=@n8, PropertyName:Name=Get L_PropertyName() -> (L_PropertyName:Name=None)) -> (then:Exec, PropertyValue:"Curve Table Row Handle Structure (by ref)", ReturnValue:bool)
   n10 = Break Curve Table Row Handle!(CurveTableRowHandle:"Curve Table Row Handle Structure (by ref)"=@n9) -> (CurveTable:Curve Table Object Reference, RowName:Name) {is=title K2Node_BreakStruct}
   > Set L_ResourceRowName(L_ResourceRowName:Name=@n10.RowName) -> (Output_Get:Name=None)
   n11 = GetSoftObjectPath(Object=@n10)
   > Branch(Condition="Not Equal (SoftObjectReference)"(
     A:"Soft Object Reference (by ref)"=To Soft Object Reference(SoftObjectPath=@n11)
     B:"Soft Object Reference (by ref)"=Get CurveTable() -> (CurveTable:Soft Object Reference)
    ) -> (ReturnValue:bool=~) {is=title K2Node_PromotableOperator})
   > CheckoutLoadedAsset!(execute, AssetToCheckout:Object Reference=Get L_Asset() -> (L_Asset:Object Reference)) -> (then, ReturnValue:bool)
   > n12 = LoadAsset_Blocking(Asset=Get CurveTable())
   > n13 = Execute Python Script!(execute, PythonScript="out_table=in_object", in_object:Object Reference=@n12) -> (then, ReturnValue, out_table:Curve Table Object Reference) {inputs [in_object], outputs [out_table]}
   > SetEditorProperty(
    Object=Get L_BuildingActorToModify() -> (L_BuildingActorToModify:Building Actor Object Reference)
    PropertyName=Get L_PropertyName() -> (L_PropertyName=None)
    PropertyValue:Curve Table Row Handle Structure=Make Curve Table Row Handle!(CurveTable:Curve Table Object Reference=@n13.out_table, RowName:Name=Get L_ResourceRowName() -> (L_ResourceRowName:Name=None)) -> (CurveTableRowHandle:Curve Table Row Handle Structure) {is=title K2Node_MakeStruct}
   )
   > SaveLoadedAsset(AssetToSave=Get L_Asset(), bOnlyIfIsDirty=true)
  ]}
]
variables [
 CurveTable {type=Object, default=/Game/Athena/Balance/DataTables/AthenaResourceRates}
 Object {type=Object, default=null}
]
